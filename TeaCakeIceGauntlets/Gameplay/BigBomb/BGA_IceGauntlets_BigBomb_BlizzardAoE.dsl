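; Blizzard area-of-effect actor for the Ice Gauntlets BigBomb, with hotfixable radius and lifespan.
; Sphere overlap applies a damage effect, gated by team and friendly-fire checks, and removes it on exit.
; The authority optionally runs timed environment damage via async targeting, capped per frame, while remotes spawn a radial wind field one frame later.
; Death handlers stop the damage timer, finish the Niagara effect and swap looping audio for an end sound.
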(event EventBeginPlay
  ;[0.03547 0.15 0.03123 0.5] Begin Play
  (|CallParentFunction)
  (Utilities|FlowControl|Sequence
    (:then_0
      (Actor|SetLifeSpan (ScalableFloat|GetValueatLevel (Variables|Hotfix|GetRow_BlizzardLifespan)))
      (Control|SetRadius (Variables|Default|GetAudioShapeSphere) (Components|Sphere|GetScaledSphereRadius (Variables|Default|GetSphere)) :!class /Script/AudioShapes.AudioShapeSphereComponent))
    (:then_1
      (Networking|SwitchHasAuthority
        (:Authority
          (if (ScalableFloat|GetValueasBool (Variables|Hotfix|GetRow_ShouldDamageEnvironment))
            (bind _returnvalue (Utilities|Time|SetTimerbyEvent (EventDispatchers|CreateEvent :!event Custom|DoEnvironmentDamage) (ScalableFloat|GetValueatLevel (Variables|Hotfix|GetRow_DamagePeriod)) true))
            (Variables|Default|SetEnvironmentDamageTimer _returnvalue)))
        (:Remote
          (Utilities|FlowControl|DelayUntilNextFrame) ; Prevent spawning two actors on same frame for perf
          (Game|SpawnActorfromClass "/Game/Effects/Niagara/Blueprints/PhsyicsFields/FN_RadialForce.FN_RadialForce_C" (Transformation|GetActorTransform) "Undefined" "MultiplyWithRoot" 0 0.0 (ScalableFloat|GetValueatLevel (Variables|Hotfix|GetRow_BlizzardLifespan)) (* (ScalableFloat|GetValueatLevel (Variables|Hotfix|GetRow_BlizzardRadius)) (ScalableFloat|GetValueatLevel (Variables|Hotfix|GetRow_WindRadiusScale))) 1.5 0.0 0))))))

(event OnComponentBeginOverlap(Sphere)
  (params (OverlappedComponent :type Primitive%20Component%20Object%20Reference) (OtherActor :type Actor%20Object%20Reference) (OtherComp :type Primitive%20Component%20Object%20Reference) (OtherBodyIndex :type Integer) (bFromSweep :type Boolean) (SweepResult :type Hit%20Result%20Structure%20%28by%20ref%29))
  (bind _returnvalue (Ability|GetFortAbilitySystemComponent OtherActor))
  (bind _instigator (Variables|Actor|GetInstigator))
  ;[0.04013 0.15 0.07876 0.5] Overlap Begin
  (bind _returnvalue_2 (Team|AllowFriendlyFire self _instigator OtherActor))
  (Utilities|IsValid OtherActor
    (:Is%20Valid
      (Utilities|IsValid _returnvalue
        (:Is%20Valid
          (if (Math|Boolean|ORBoolean (not (Team|OnSameTeam OtherActor _instigator)) _returnvalue_2 (ScalableFloat|GetValueasBool (Variables|Hotfix|GetRow_FriendlyFireEnabled)))
            (Utilities|IsValid _returnvalue
              (:Is%20Valid
                (GameplayEffects|ApplyGameplayEffectSpecToSelf _returnvalue (GameplayEffects|MakeOutgoingSpec (Ability|GetAbilitySystemComponent 0) (Variables|Default|GetGE_BlizzardDamage) 0.0 (Class|AthenaFunctionLibrary|MakeSimpleFortEffectContext _instigator self)))))))))))

(event Custom|DoEnvironmentDamage
  (bind _returnvalue (Ability|GetAbilitySystemComponent 0))
  ;[0.0295 0.15 0.1477 0.5] Environment Damage DOT
  (bind (_asynctaskproxy _targetinghandle) (Class|AsyncActionPerformTargeting|PerformTargetingAsyncAction 0 (Variables|Default|GetEnvironmentDamageTargetingPreset) true)
    (:Targeted
      (Utilities|IsValid _returnvalue
        (:Is%20Valid
          (Ability|Tasks|ApplyGameplayEffectSpecToTargetingHandleAsync _returnvalue _targetinghandle (GameplayEffects|MakeOutgoingSpec _returnvalue (Variables|Default|GetGE_EnvironmentDamage) 0.0 (Class|AthenaFunctionLibrary|MakeSimpleFortEffectContext (Variables|Actor|GetInstigator) self)) (Variables|Hotfix|GetRow_MaxDamageApplicationsPerFrame)))))))

(event OnComponentEndOverlap(Sphere)
  (params (OverlappedComponent :type Primitive%20Component%20Object%20Reference) (OtherActor :type Actor%20Object%20Reference) (OtherComp :type Primitive%20Component%20Object%20Reference) (OtherBodyIndex :type Integer))
  (bind _returnvalue (Ability|GetFortAbilitySystemComponent OtherActor))
  ;[0.15 0.07483 0.04279 0.5] Overlap End
  (Utilities|IsValid _returnvalue
    (:Is%20Valid
      (GameplayEffects|RemoveActiveGameplayEffectbySourceEffect _returnvalue "/TeaCakeIceGauntlets/Gameplay/BigBomb/GE_IceGauntlets_BigBomb_BlizzardDamage.GE_IceGauntlets_BigBomb_BlizzardDamage_C" 0 -1))))

(event EventOnDeathPlayEffects
  (params (Damage :type Float%20%28single-precision%29) (DamageTags :type Gameplay%20Tag%20Container%20Structure%20%28by%20ref%29) (Momentum :type Vector :default "0, 0, 0") (HitInfo :type Hit%20Result%20Structure%20%28by%20ref%29) (InstigatedBy :type Fort%20Pawn%20Object%20Reference) (DamageCauser :type Actor%20Object%20Reference) (EffectContext :type Gameplay%20Effect%20Context%20Handle%20Structure))
  (bind _ns_firepetal_snowgrenade_blizzard_aoe (Variables|Default|GetNS_FirePetal_SnowGrenade_Blizzard_AOE))
  ;[0.15 0.03662 0.08554 0.5] Death FX
  (Components|DestroyComponent (Variables|Default|GetAudioShapeSphere))
  (Audio|PlaySoundatLocation "/TeaCakeIceGauntlets/Assets/Audio/SC_TeaCake_IceGauntlets_Blizzard_End.SC_TeaCake_IceGauntlets_Blizzard_End" (Transformation|GetActorLocation))
  (Niagara|SetNiagaraVariable(Bool) _ns_firepetal_snowgrenade_blizzard_aoe "Finish" true)
  (Utilities|FlowControl|Delay 1.5)
  (Components|DestroyComponent _ns_firepetal_snowgrenade_blizzard_aoe))

(event EventOnDeathServer
  (params (Damage :type Float%20%28single-precision%29) (DamageTags :type Gameplay%20Tag%20Container%20Structure%20%28by%20ref%29) (Momentum :type Vector :default "0, 0, 0") (HitInfo :type Hit%20Result%20Structure%20%28by%20ref%29) (InstigatedBy :type Controller%20Object%20Reference) (DamageCauser :type Actor%20Object%20Reference) (EffectContext :type Gameplay%20Effect%20Context%20Handle%20Structure))
  (bind _environmentdamagetimer (Variables|Default|GetEnvironmentDamageTimer))
  (bind _ns_firepetal_snowgrenade_blizzard_aoe (Variables|Default|GetNS_FirePetal_SnowGrenade_Blizzard_AOE))
  ;[0.15 0.03662 0.08554 0.5] Death FX
  (Utilities|Time|PauseTimerbyHandle _environmentdamagetimer)
  (Utilities|Time|ClearandInvalidateTimerbyHandle _environmentdamagetimer)
  (Niagara|SetNiagaraVariable(Bool) _ns_firepetal_snowgrenade_blizzard_aoe "Finish" true)
  (Utilities|FlowControl|Delay 1.5)
  (Components|DestroyComponent _ns_firepetal_snowgrenade_blizzard_aoe))

(fn ConstructionScript
  (Components|Sphere|SetSphereRadius (Variables|Default|GetSphere) (ScalableFloat|GetValueatLevel (Variables|Hotfix|GetRow_BlizzardRadius)) true))
