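# EQS query that picks locations for icy patches spawned by the Big Bomb in TeaCake Ice Gauntlets.
# A composite generator merges three circle generators into one candidate set.
# An overlap test filters and scores points with a raised box, alongside a valid-surface check for ground flatness.
# An obstruction test filters points by tracing from the querier, and all three tests use the same custom game trace channel.
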EnvQuery /TeaCakeIceGauntlets/Gameplay/BigBomb/EQS_BigBomb_SpawnIcyPatches
options [{
  generator {
   name=EnvQueryGenerator_Composite_0
   class=EnvQueryGenerator_Composite
   properties {
    bCanRunAsync=true
    optionName=""
    generators [
     &self:EnvQueryGenerator_Composite_0.EnvQueryGenerator_OnCircle_0
     &self:EnvQueryGenerator_Composite_0.EnvQueryGenerator_OnCircle_2
     &self:EnvQueryGenerator_Composite_0.EnvQueryGenerator_OnCircle_1
    ]
   }
  }
  tests [
   EnvQueryTest_Overlap_4 {
    class=EnvQueryTest_Overlap
    properties {testPurpose=FilterAndScore, overlapData {extentX=75, extentY=75, extentZ=75, overlapChannel=ECC_GameTraceChannel11, shapeOffset=<0 0 100>}}
   }
   FortQueryTest_ValidSurface_4 {
    class=FortQueryTest_ValidSurface
    properties {
     flatSurfaceToleranceZ=75
     radius=100
     traceCollisionChannel=ECC_GameTraceChannel11
     traceOffsetDown=100
     traceOffsetUp=100
     surfaceMaterials []
    }
   }
   FortQueryTest_IsObstructed_4 {
    class=FortQueryTest_IsObstructed
    properties {
     context=&EnvQueryContext_Querier
     contextHeightOffset=10
     testPurpose=Filter
     traceFromContext=true
     traceData {serializedChannel=ECC_GameTraceChannel11, traceChannel=TraceTypeQuery10}
    }
   }
  ]
 }]
